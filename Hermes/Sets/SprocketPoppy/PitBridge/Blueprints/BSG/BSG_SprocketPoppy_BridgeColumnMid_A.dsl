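; BSG_SprocketPoppy_BridgeColumnMid_A is a Blueprint for a mid bridge column piece in the SprocketPoppy PitBridge set.
; Its construction script branches on the SpawnMasks flag to handle the ISM_SocketFiller component.
; When the flag is set it reconstructs the socket filler; otherwise it destroys that component.
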(fn ConstructionScript
  (if (Variables|Default|GetSpawnMasks)
    (Class|ISMSocketFiller|ReconstructSocketFiller (Variables|Default|GetISM_SocketFiller))
    (else
      (Components|DestroyComponent (Variables|Default|GetISM_SocketFiller)))))
